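# Unlit material that fakes the interior seen in the Battle Bus booster's hatch, with Nanite usage enabled.
# A constant-height BumpOffset shifts UVs for parallax depth, then offset and center-scale parameters position the BusHatchInteriorMask texture.
# The power-shaped red channel shades the wall color, and the power-shaped green channel blends in the lights color.
# The shared MF_PartyBusAnimationWPO function supplies party bus vertex animation.
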Material /Game/Athena/Aircraft/BattleBus_Booster/M_BusHatchInterior
graph {
 properties {bUsedWithNanite=true, pixelDepthOffsetMode=PDOM_Legacy, shadingModel=MSM_Unlit, usedShadingModels=Unlit}
 nodes [
  n1 = BumpOffset(
   Height=ScalarParameter() {defaultValue=5, parameterName=Height}
   HeightRatioInput=ScalarParameter() {defaultValue=-0.11, parameterName=HeightRatio}
  ) {heightRatio=-0.1}
  n2 = AppendVector(A=ScalarParameter() {parameterName=OffsetU}, B=ScalarParameter() {parameterName=OffsetV})
  n3 = TextureSample(UVs=MaterialFunctionCall(UVs=Add(A=@n1, B=@n2), Texture Scale=ScalarParameter() {defaultValue=1.8, parameterName=Scale}) {materialFunction=&ScaleUVsByCenter}.UVs) {samplerType=SAMPLERTYPE_LinearColor, texture=&/BusHatchInteriorMask}
  n4 = ScalarParameter() {defaultValue=3, parameterName=RoomPower}
  LinearInterpolate(
   A=Multiply(A=VectorParameter() {parameterName=ColorWall, defaultValue=<rgba 0.0521 0.0127 0.0122 1>}.RGB, B=Power(Base=@n3.R, Exp=@n4))
   B=VectorParameter() {parameterName=ColorLights, defaultValue=<rgba 0.0025 0.0004 0.0003 1>}.RGB
   Alpha=Power(Base=@n3.G, Exp=ScalarParameter() {defaultValue=8, parameterName=LightPower})
  )
  MaterialFunctionCall() {materialFunction=/Game/Athena/Aircraft/PartyBus_RT/Materials/MF_PartyBusAnimationWPO}
 ]
}
